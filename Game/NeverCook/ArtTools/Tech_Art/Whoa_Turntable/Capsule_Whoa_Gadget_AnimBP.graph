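# Anim Blueprint for the Logarithm Whoa glasses gadget in a tech-art turntable folder, fitting the prop to the owning player character.
# On initialization it looks up an attach row by character plus variant, then character, then default male/female Commando.
# The row selects the part mesh and socket to attach to, and whether glasses show.
# AnimGraph applies a fixed main-bone transform, shrinking it when disabled; row offsets are stored but unapplied.
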AnimBlueprint /Game/NeverCook/ArtTools/Tech_Art/Whoa_Turntable/Capsule_Whoa_Gadget_AnimBP
target_skeleton=/Game/Animation/Props/Gadgets_Logarithm_Whoa/Meshes/Logarithm_Whoa_Gadget_Skeleton
graphs [
 AnimGraph {nodes [
   n1 = "Transform (Modify) Bone - Bone: main"(ComponentPose=Local To Component(LocalPose=Input Pose()), Alpha=1, Translation="0,9,6.5", Rotation="0,0,-10", Scale="1.5,1.5,1.5") {
    rotationMode=BMM_Replace
    rotationSpace=BCS_ParentBoneSpace
    scaleMode=BMM_Replace
    scaleSpace=BCS_ParentBoneSpace
    translationMode=BMM_Replace
    translationSpace=BCS_ParentBoneSpace
    alphaBoolBlend {blendOption=Linear}
    boneToModify {boneName=main}
    rotation {roll=-10}
    scale=<1.5 1.5 1.5>
    translation {y=9, z=6.5}
   }
   n2 = "Transform (Modify) Bone - Bone: main"!(
    ComponentPose=@n1
    bAlphaBoolEnabled:bool=Get Glasses Disabled() -> (Glasses Disabled:bool)
    Translation="0,0,0"
    Rotation="0,0,0"
    Scale="0.01,0.01,0.01"
   ) -> (Pose) {
    alphaInputType=Bool
    rotationSpace=BCS_ParentBoneSpace
    scaleMode=BMM_Replace
    scaleSpace=BCS_ParentBoneSpace
    translationSpace=BCS_ParentBoneSpace
    alphaBoolBlend {blendOption=Linear}
    boneToModify {boneName=main}
    scale=<0.01 0.01 0.01>
   }
   Output Pose(Result=Component To Local(ComponentPose=@n2)) {layerGroup=None}
   Get Local Offset T() -> (Local Offset T:Vector)
   Get Local Offset S() -> (Local Offset S:Vector)
   Get Local Offset R() -> (Local Offset R:Rotator)
  ]}
 EventGraph {nodes [
   Event BlueprintInitializeAnimation()
   > n1 = Sequence!(execute) -> (then_0, then_1, then_2:Exec, then_3:Exec)
   n2 = Sequence!(execute=@n1.then_2) -> (then_0, then_1, then_2)
   n3 = Break Emote Prop Attach Struct!(EmotePropAttachStruct:"Emote Prop Attach Struct Structure (by ref)"=Get Attach Data Row() -> (Attach Data Row:Emote Prop Attach Struct Structure)) -> (
    enabled_18_#1:bool=~
    socket_2_#2:Name
    offsetT_6_#3:Vector=~
    offsetR_9_#4:Rotator=~
    offsetS_15_#5:Vector=~
    part_21_#6:EFortCustomPartType Enum
    variantTag_26_#7:Gameplay Tag Structure
    notes_33_#8:String
   ) {is=title K2Node_BreakStruct}
   > Set Glasses Disabled(Glasses Disabled:bool=NOT Boolean(A=@n3)) -> (Output_Get:bool)
   > n4 = Set Local Offset T(Local Offset T:Vector=@n3.offsetT_6_#3) -> (Output_Get:Vector)
   > n5 = Set Local Offset R(Local Offset R:Rotator=@n3.offsetR_9_#4) -> (Output_Get:Rotator)
   > n6 = Set Local Offset S(Local Offset S:Vector=@n3.offsetS_15_#5) -> (Output_Get:Vector)
   n7 = Get Fort Player Pawn() -> (Fort Player Pawn:Fort Player Pawn Object Reference)
   Is Valid(exec=@n2.then_1, InputObject=@n7)
   n8 = GetSkeletalMeshForPartType(self=@n7, PartType=@n3.part_21_#6)
   > Is Valid(InputObject=@n8)
   > Set Driving Skelmesh(Driving Skelmesh:Skeletal Mesh Component Object Reference=@n8) -> (Output_Get:Skeletal Mesh Component Object Reference)
   > Set Socket(Socket:Name=@n3.socket_2_#2) -> (Output_Get:Name=None)
   n9 = Cast To FortPlayerPawn(execute=@n1.then_0, Object=GetOwningActor())
   > Set Fort Player Pawn(Fort Player Pawn:Fort Player Pawn Object Reference=@n9) -> (Output_Get:Fort Player Pawn Object Reference)
   > Set currentCharacter(currentCharacter:String=GetCharacterDisplayName!(Actor:Actor Object Reference=@n9) -> (ReturnValue:String)) -> (Output_Get:String)
   n10 = GetAllLoadoutVariantInfoFromActor!(Actor:Actor Object Reference=@n9) -> (ReturnValue:Array of Mcp Variant Channel Info Structures)
   n11 = GetDebugStringFromGameplayTag!(GameplayTag:Gameplay Tag Structure=Break Mcp Variant Channel Info!(McpVariantChannelInfo:"Mcp Variant Channel Info Structure (by ref)"="Get (a copy)"(Array:Array of Mcp Variant Channel Info Structures=@n10) -> (Output:Mcp Variant Channel Info Structure)) -> (
     OwnedVariantTags:Gameplay Tag Container Structure
     ItemVariantIsUsedFor:Item Definition Base Object Reference
     CustomData:String
     VariantChannelTag:Gameplay Tag Structure
     ActiveVariantTag:Gameplay Tag Structure
    ) {is=title K2Node_BreakStruct}.ActiveVariantTag) -> (ReturnValue:String)
   > Set currentVariant(currentVariant:String=RightChop!(SourceString:String=@n11, Count:int=FindSubstring(SearchIn=@n11, Substring=., bSearchFromEnd=true)) -> (ReturnValue:String)) -> (Output_Get:String)
   Attach Component To Component!(
    execute=@n1.then_3
    self:Scene Component Object Reference=GetOwningComponent()
    Parent:Scene Component Object Reference=Get Driving Skelmesh() -> (Driving Skelmesh:Skeletal Mesh Component Object Reference)
    SocketName:Name=Get Socket() -> (Socket:Name=None)
    LocationRule:EAttachmentRule Enum=KeepRelative
    RotationRule:EAttachmentRule Enum=KeepRelative
    ScaleRule:EAttachmentRule Enum=KeepRelative
    bWeldSimulatedBodies:bool=TRUE
   ) -> (then, ReturnValue:bool)
   n12 = GetDataTableRowNames(execute=@n1.then_1, Table=Get Data Table() -> (Data Table:Data Table Object Reference))
   n13 = String To Name(InString=Append(A=Get currentCharacter() -> (currentCharacter:String), B=Get currentVariant() -> (currentVariant:String)))
   > n14 = Branch(Condition="integer > integer"(A=Find Item!(TargetArray:Array of Names=@n12, ItemToFind:"Name (by ref)"=@n13) -> (ReturnValue), B=-1) {is=title K2Node_CallFunction})
   > n15 = Get Data Table Row!(execute:Exec, DataTable:Data Table Object Reference=Get Data Table(), RowName:Name=@n13) -> (then:Exec, RowNotFound:Exec, ReturnValue:Emote Prop Attach Struct Structure)
   > Set Attach Data Row(Attach Data Row:Emote Prop Attach Struct Structure=@n15) -> (Output_Get:Emote Prop Attach Struct Structure)
   n16 = GetDataTableRowNames(execute=@n14.else, Table=Get Data Table())
   n17 = Find Item(TargetArray=@n16, ItemToFind=String To Name(InString=Get currentCharacter()))
   > n18 = Branch(Condition="integer > integer"(A=@n17, B=-1))
   > n19 = Get Data Table Row(DataTable=Get Data Table(), RowName="Get (a ref)"(Array:Array of Names=@n16, Dimension 1=@n17) -> (Output:"Name (by ref)"))
   > Set Attach Data Row(Attach Data Row=@n19)
   n20 = Get currentCharacter()
   n21 = Branch(execute=@n18.else, Condition="integer > integer"(A=FindSubstring(SearchIn=@n20, Substring=Commando_F), B=-1))
   > n22 = Get Data Table Row(DataTable=Get Data Table(), RowName=CID_001_Athena_Commando_F_Default)
   > Set Attach Data Row(Attach Data Row=@n22)
   n23 = Get Data Table Row(execute=@n21.else, DataTable=Get Data Table(), RowName=CID_005_Athena_Commando_M_Default)
   > Set Attach Data Row(Attach Data Row=@n23)
   TryGetPawnOwner()
   MakeTransform(Location=@n4, Rotation=@n5, Scale=@n6)
   Event BlueprintUpdateAnimation()
  ]}
]
variables [
 Attach Data Row {
  type=EmotePropAttachStruct
  default {enabled=true, notes="", part=Head, socket=head, offsetR=<rot 0 0 0>, offsetS=<1 1 1>, offsetT=<0 0 0>, variantTag {tagName=None}}
 }
 currentCharacter {type=string, default=""}
 currentVariant {type=string, default=""}
 Data Table {type=DataTable, default=/Game/Animation/Props/Gadgets_Logarithm_Whoa/Meshes/Logarithm_Whoa_Gadget_Attach_DataTable}
 Driving Skelmesh {type=SkeletalMeshComponent}
 Fort Player Pawn {type=FortPlayerPawn}
 Glasses Disabled {type=boolean}
 Local Offset R {type=Rotator}
 Local Offset S {type=Vector}
 Local Offset T {type=Vector}
 Socket {type=string, default=None}
]
