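; TestWhoa is an actor Blueprint in the Whoa_Turntable tech-art tools folder, apparently a test asset, that reports a character ID and variant.
; Its one implemented function, Report_CID_Variant, returns the CharacterID and Variant variables.
; The BeginPlay, Tick and overlap events and the construction script contain no logic.
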(event EventBeginPlay)

(event Collision|EventActorBeginOverlap
  (params (OtherActor :type Actor%20Object%20Reference)))

(event EventTick
  (params (DeltaSeconds :type Float%20%28single-precision%29)))

(fn Report_CID_Variant
  (return (Variables|Default|GetCharacterID) (Variables|Default|GetVariant)))

(fn ConstructionScript)
