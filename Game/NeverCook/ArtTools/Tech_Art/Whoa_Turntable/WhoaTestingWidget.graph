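# WhoaTestingWidget is a tech-art Editor Utility Widget that batch-screenshots Fortnite character cosmetics performing an emote in PIE.
# It loads character definitions from a named collection into CIDList and pushes the indexed entry into an embedded CosmeticPickerv2.
# Emotes fire through the PlayEmoteItem cheat, then HighResShot saves files named by index, character and suffix.
# RunAutomation re-calls itself with delays, incrementing the index until past list length or force-stopped.
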EditorUtilityWidgetBlueprint /Game/NeverCook/ArtTools/Tech_Art/Whoa_Turntable/WhoaTestingWidget
graphs [EventGraph {nodes [
   n1 = "On Clicked (ManualEmote)"!() -> (OutputDelegate, then, widget:User Widget Object Reference, event:Pointer Event Structure) {is=title K2Node_ComponentBoundEvent}
   n2 = GetPIEWorlds!(execute=[@n1.then, @n22.then], bIncludeDedicatedServer:bool=false) -> (then, ReturnValue:Array of World Object References)
   > n3 = For Each Loop(Array:Array of World Object References=@n2) -> (Array Element:World Object Reference)
   n4 = Get CurrentEmote() -> (CurrentEmote:Athena Dance Item Definition Object Reference)
   > ExecuteConsoleCommand(
    WorldContextObject=@n3
    Command=Append(A="cheat PlayEmoteItem ", B=GetDisplayName(Object=@n4))
    SpecificPlayer=GetPlayerController!(WorldContextObject:Object Reference=@n3, PlayerIndex:int=0) -> (ReturnValue:Player Controller Object Reference)
   )
   "On Clicked (PopulateAllCID)"!() -> (OutputDelegate, then, widget:User Widget Object Reference, event:Pointer Event Structure) {is=title K2Node_ComponentBoundEvent}
   > Set StartIndex(StartIndex:int=0) -> (Output_Get:int=0)
   > Clear(TargetArray:Array of Athena Character Item Definition Object References=Get CIDList() -> (CIDList:Array of Athena Character Item Definition Object References))
   > n5 = GetAssetsInCollection!(
    execute
    self:Asset Tags Subsystem Object Reference=Get AssetTagsSubsystem() -> (ReturnValue:Asset Tags Subsystem Object Reference) {is=title K2Node_GetEngineSubsystem}
    Name:Name=Get PopulateCID_CollectionName() -> (PopulateCID_CollectionName:Name=None)
   ) -> (then, ReturnValue:Array of Asset Data Structures)
   > PrintString(InString="To String (Integer)"(InInt=Length(TargetArray:Array of Asset Data Structures=@n5)))
   > RunAssetsThroughFilter!(
    execute:Exec
    self:Object Reference="To Object (Interface)"!(Interface:"Interface Interface (by ref)"=GetAssetRegistry()) -> (ReturnValue:Object Reference)
    AssetDataList:Array of Asset Data Structures=@n5
    Filter:"ARFilter Structure (by ref)"=MakeARFilter!(
     PackageNames:Array of Names
     PackagePaths:Array of Names
     SoftObjectPaths:Array of Soft Object Path Structures
     ClassPaths:Array of Top Level Asset Path Structures
     RecursiveClassPathsExclusionSet:Set of Top Level Asset Path Structures
     ClassNames:Array of Names=Make Array("[0]":Name=AthenaCharacterItemDefinition) -> (Array:Array of Names)
     RecursiveClassesExclusionSet:Set of Names
     bRecursivePaths:bool=false
     bRecursiveClasses:bool=false
     bIncludeOnlyOnDiskAssets:bool=false
    ) -> (ReturnValue:ARFilter Structure)
   ) -> (then:Exec) {is=title K2Node_Message}
   > n6 = For Each Loop(Array:Array of Asset Data Structures=@n5) -> (Array Element:Asset Data Structure)
   > n7 = LoadAsset(AssetPath=ToSoftObjectPath!(InAssetData:"Asset Data Structure (by ref)"=@n6) -> (ReturnValue_PathString:String))
   > n8 = Cast To AthenaCharacterItemDefinition!(execute, Object=@n7) -> (then, CastFailed, AsAthena Character Item Definition:Athena Character Item Definition Object Reference) {is=title K2Node_DynamicCast}
   > Add Unique(
    TargetArray:Array of Athena Character Item Definition Object References=Get CIDList()
    NewItem:"Athena Character Item Definition Object Reference (by ref)"=@n8
   )
   PrintString(execute=@n8.CastFailed, InString=Failed to Cast)
   Reroute Node(InputPin=@n6.Completed)
   n9 = "On Clicked (NextCID)"!() -> (OutputDelegate, then, widget:User Widget Object Reference, event:Pointer Event Structure) {is=title K2Node_ComponentBoundEvent}
   n10 = Set CurrentCID(
    execute=[@n9.then, @n20.then]
    CurrentCID:Athena Character Item Definition Object Reference="Get (a copy)"(Array:Array of Athena Character Item Definition Object References=Get CIDList(), Dimension 1=Get StartIndex() -> (StartIndex:int=0)) -> (Output:Athena Character Item Definition Object Reference)
   ) -> (Output_Get:Athena Character Item Definition Object Reference)
   > Set CIDName(CIDName:String=GetDisplayName(Object=@n10)) -> (Output_Get:String)
   > n11 = getTabWidgetByIndex!(
    execute
    self:Base Tab Container Widget Object Reference=Get Generic01_TabContainer!(self:Cosmetic Pickerv 2 Object Reference=Get CosmeticPickerv2() -> (CosmeticPickerv2:Cosmetic Pickerv 2 Object Reference)) -> (Generic01_TabContainer:Generic 01 Tab Container Object Reference)
    index:int=0
   ) -> (then, widget:Widget Object Reference)
   > n12 = Cast To EUW_CosmeticPickerv2!(execute, Object=@n11) -> (then, CastFailed, AsEUW Cosmetic Pickerv 2:EUW Cosmetic Pickerv 2 Object Reference) {is=title K2Node_DynamicCast}
   n13 = Get EUW_CosmeticPicker_ID_Character!(self:EUW Cosmetic Pickerv 2 Object Reference=@n12) -> (EUW_CosmeticPicker_ID_Character:EUW Cosmetic Pickerv 2 ID Object Reference)
   > Set character!(execute, character:Athena Character Item Definition Object Reference=@n10, self:EUW Cosmetic Pickerv 2 ID Object Reference=@n13) -> (then, Output_Get:Athena Character Item Definition Object Reference)
   > IDChanged!(execute, self:EUW Cosmetic Pickerv 2 ID Object Reference=@n13, PropertyName:Name=None) -> (then)
   n14 = "On Clicked (SaveScreenshot)"!() -> (OutputDelegate, then, widget:User Widget Object Reference, event:Pointer Event Structure) {is=title K2Node_ComponentBoundEvent}
   n15 = GetDisplayName(Object=Get CurrentCID() -> (CurrentCID:Athena Character Item Definition Object Reference))
   PrintString(execute=[@n14.then, @n21.then], InString=@n15)
   n16 = "To Text (Integer)"(Value=Get StartIndex() -> (StartIndex=0), MinimumIntegralDigits=4)
   > ExecuteConsoleCommand(Command=Append!(
     A="HighResShot 1920x1080 filename=\""
     B=Get!() -> (Screenshot Directory_Path:String)
     C:String=/
     D:String="To String (Text)"(InText=@n16)
     E:String=_
     F:String=@n15
     G:String=Get FilenameSuffix() -> (FilenameSuffix:String)
     H:String=".png\""
    ) -> (ReturnValue))
   n17 = "On Clicked (EmoteAndRecord)"!() -> (OutputDelegate, then, widget:User Widget Object Reference, event:Pointer Event Structure) {is=title K2Node_ComponentBoundEvent}
   ForceEmote(execute=[@n17.then, @n23.then])
   > Delay!(execute, WorldContextObject:Object Reference, Duration=0.5) -> (then)
   > TakeScreenshot()
   n18 = "On Selection Changed (EnableAutomation)"!() -> (OutputDelegate, then, selected:bool, widget:User Widget Object Reference) {is=title K2Node_ComponentBoundEvent}
   > SetIsEnabled(
    self=[
     Get AutomatedSwitchEmoteRecord() -> (AutomatedSwitchEmoteRecord:TA Generic 01 Button Object Reference)
     Get EmergencyStopAutomation() -> (EmergencyStopAutomation:TA Generic 01 Checkbox Object Reference)
    ]
    bInIsEnabled=@n18.selected
   )
   "On Clicked (AutomatedSwitchEmoteRecord)"!() -> (OutputDelegate, then, widget:User Widget Object Reference, event:Pointer Event Structure) {is=title K2Node_ComponentBoundEvent}
   > RunAutomation()
   n19 = "On Selection Changed (EmergencyStopAutomation)"!() -> (OutputDelegate, then, selected:bool, widget:User Widget Object Reference) {is=title K2Node_ComponentBoundEvent}
   > Set ForceStopAutomation?(ForceStopAutomation?:bool=@n19.selected) -> (Output_Get:bool)
   n20 = Internal_NextCID() {is=title K2Node_CustomEvent}
   n21 = TakeScreenshot() {is=title K2Node_CustomEvent}
   n22 = ForceEmote() {is=title K2Node_CustomEvent}
   n23 = Internal_EmoteAndRecord() {is=title K2Node_CustomEvent}
   n24 = RunAutomation() {is=title K2Node_CustomEvent}
   > Internal_NextCID() {is=title K2Node_CallFunction}
   > Delay(Duration=5)
   > Internal_EmoteAndRecord() {is=title K2Node_CallFunction}
   > Delay(Duration=2)
   n25 = Get StartIndex() -> (StartIndex=0)
   > Increment Int(Value=@n25)
   n26 = Get CIDList()
   n27 = OR Boolean(
    A="integer > integer"(A=@n25, B=Length(TargetArray:Array of Athena Character Item Definition Object References=@n26)) {is=title K2Node_CallFunction}
    B=Get ForceStopAutomation?() -> (ForceStopAutomation?:bool)
   )
   > n28 = Branch(Condition=@n27)
   RunAutomation(execute=@n28.else) {is=title K2Node_CallFunction}
   n29 = BindOnCIDUpdated!() -> (OutputDelegate, then, PropertyName:Name) {is=title K2Node_CustomEvent}
   > n30 = getTabWidgetByIndex(self=Get Generic01_TabContainer(self=Get CosmeticPickerv2()), index=0)
   > n31 = Cast To EUW_CosmeticPickerv2(Object=@n30)
   > Set CurrentCID(CurrentCID=Get character!(self:EUW Cosmetic Pickerv 2 ID Object Reference=Get EUW_CosmeticPicker_ID_Character(self=@n31)) -> (character:Athena Character Item Definition Object Reference))
   Event PreConstruct()
   n32 = Self-Reference()
   > SetObject(self=Get AllCIDS() -> (AllCIDS:Details View Object Reference), NewObject=@n32)
   > SetObject(self=Get CurrentArrayPos() -> (CurrentArrayPos:Details View Object Reference), NewObject=@n32)
   > SetObject(self=Get Emote() -> (Emote:Details View Object Reference), NewObject=@n32)
   > n33 = getTabWidgetByIndex(self=Get Generic01_TabContainer(self=Get CosmeticPickerv2()), index=0)
   > n34 = Cast To EUW_CosmeticPickerv2(Object=@n33)
   n35 = Get EUW_CosmeticPicker_ID_Character(self=@n34)
   > Set CurrentCID(CurrentCID=Get character(self=@n35))
   > Bind Event to On Property Changed(
    self:Property View Base Object Reference=Get DetailsView_ID!(self:EUW Cosmetic Pickerv 2 ID Object Reference=@n35) -> (DetailsView_ID:Details View Object Reference)
    Delegate=@n29
   ) {is=title K2Node_AddDelegate}
   Reroute Node()
   > n36 = Cast To CosmeticPickerv2!(execute, Object=Get CosmeticPickerv2()) -> (then, CastFailed, AsCosmetic Pickerv 2:Cosmetic Pickerv 2 Object Reference) {is=title K2Node_DynamicCast}
   > n37 = getTabWidgetByIndex(self=Get Generic01_TabContainer(self=@n36), index=0)
   > n38 = Cast To EUW_CosmeticPickerv2(Object=@n37)
   > IDChanged(self=Get EUW_CosmeticPicker_ID_Character(self=@n38), PropertyName=None)
   "Equal (Class)"!(
    A:Object Class Reference=GetClass!(InAssetData:"Asset Data Structure (by ref)"=@n6) -> (ReturnValue:Object Class Reference)
    B:Object Class Reference=/Script/FortniteGame.AthenaCharacterItemDefinition
   ) -> (ReturnValue:bool) {is=title K2Node_CallFunction}
   GetAssetRegistry()
   Clear and Invalidate Timer by Handle!(execute, WorldContextObject:Object Reference, Handle=Set Timer by Event(Delegate=@n24, Time=7, bLooping=true)) -> (then)
   Update Text!(
    execute
    self:Whoa Testing BP Object Reference=Get UI Ref() -> (UI Ref:Whoa Testing BP Object Reference)
    Name:String=Get CIDName() -> (CIDName:String)
    Date:String
    build:String
    objectList:String
   ) -> (then)
   NOT Boolean(A=@n27)
   "On Property Changed (Emote)"!() -> (OutputDelegate, then, PropertyName:Name=None) {is=title K2Node_ComponentBoundEvent}
   "On Property Changed (CurrentArrayPos)"!() -> (OutputDelegate, then, PropertyName:Name=None) {is=title K2Node_ComponentBoundEvent}
   Internal_Emote() {is=title K2Node_CustomEvent}
   Event Construct()
   Event Tick()
  ]}]
variables [
 CIDList {type="array<AthenaCharacterItemDefinition>", default []}
 CIDName {type=string, default=Default}
 CurrentCID {type=AthenaCharacterItemDefinition, default=null}
 CurrentEmote {type=AthenaDanceItemDefinition, default=&/EID_CapsuleWhoa_Testing}
 FilenameSuffix {type=string, default=""}
 ForceStopAutomation? {type=boolean}
 PopulateCID_CollectionName {type=string, default=Audit_InCook_BROnly}
 Results {type="array<AssetData>", default []}
 Screenshot Directory {type=DirectoryPath, default {path=E:/RenderHere/WhoaImages}}
 StartIndex {type=integer}
 UI Ref {type=WhoaTesting_BP_C, default=null}
]
widgets {
 info {rootWidgetClass=&Overlay, widgetCount=21}
 widgets [
  Overlay_0 {widgetClassPath=&Overlay, widget=Overlay_0}
  VerticalBox_0 {
   widgetClassPath=&VerticalBox
   parent=Overlay_0
   widget=VerticalBox_0
   slot {class=OverlaySlot, properties {horizontalAlignment=HAlign_Fill, verticalAlignment=VAlign_Fill}}
  }
  HorizontalBox_0 {
   widgetClassPath=&HorizontalBox
   parent=VerticalBox_0
   widget=HorizontalBox_0
   slot {class=VerticalBoxSlot, properties {verticalAlignment=VAlign_Top, size {sizeRule=Fill, value=1}}}
  }
  VerticalBox_1 {
   widgetClassPath=&VerticalBox
   parent=HorizontalBox_0
   widget=VerticalBox_1
   slot {class=HorizontalBoxSlot, properties {size {sizeRule=Fill, value=1}}}
  }
  ManualEmote {
   widgetClassPath=&TA_Generic01_Button
   parent=VerticalBox_1
   widget=ManualEmote
   properties {label=EMOTE}
   slot {class=VerticalBoxSlot}
  }
  SaveScreenshot {
   widgetClassPath=&TA_Generic01_Button
   parent=VerticalBox_1
   widget=SaveScreenshot
   properties {label=Screenshot, visibility=Visible}
   slot {class=VerticalBoxSlot}
  }
  PopulateAllCID {
   widgetClassPath=&TA_Generic01_Button
   parent=VerticalBox_1
   widget=PopulateAllCID
   properties {label=Populate CIDList From Collection, visibility=Visible}
   slot {class=VerticalBoxSlot}
  }
  Add3DOverlay {
   widgetClassPath=&TA_Generic01_Button
   parent=VerticalBox_1
   widget=Add3DOverlay
   properties {bIsEnabled=false, label="(During PIE) Add 3D Overlay", visibility=Collapsed, buttonColor_Default=<rgba 1 0 0.0084 1>}
   slot {class=VerticalBoxSlot}
  }
  NextCID {
   widgetClassPath=&TA_Generic01_Button
   parent=VerticalBox_1
   widget=NextCID
   properties {label=Next CID}
   slot {class=VerticalBoxSlot}
  }
  EmoteAndRecord {
   widgetClassPath=&TA_Generic01_Button
   parent=VerticalBox_1
   widget=EmoteAndRecord
   properties {label=Emote and Screenshot}
   slot {class=VerticalBoxSlot}
  }
  Spacer_120 {
   widgetClassPath=&Spacer
   parent=VerticalBox_1
   widget=Spacer_120
   slot {class=VerticalBoxSlot, properties {padding=<ltrb 10 10 10 10>}}
  }
  AutomatedSwitchEmoteRecord {
   widgetClassPath=&TA_Generic01_Button
   parent=VerticalBox_1
   widget=AutomatedSwitchEmoteRecord
   properties {bIsEnabled=false, label="Automated Switch, Emote and Record", buttonColor_Default=<rgba 1 0.8928 0 1>}
   slot {class=VerticalBoxSlot}
  }
  HorizontalBox_1 {widgetClassPath=&HorizontalBox, parent=VerticalBox_1, widget=HorizontalBox_1, slot {class=VerticalBoxSlot}}
  EnableAutomation {
   widgetClassPath=/Game/Blueprints/Blutilities/ForTechArt/Widgets/Buttons/TA_Generic01_Checkbox
   bIsVariable=true
   parent=HorizontalBox_1
   widget=EnableAutomation
   properties {label=Enable Automation}
   slot {class=HorizontalBoxSlot, properties {size {sizeRule=Fill, value=1}}}
  }
  EmergencyStopAutomation {
   widgetClassPath=&TA_Generic01_Checkbox
   bIsVariable=true
   parent=HorizontalBox_1
   widget=EmergencyStopAutomation
   properties {bIsEnabled=false, label=Force Stop Automation Run}
   slot {class=HorizontalBoxSlot, properties {size {sizeRule=Fill, value=1}}}
  }
  ScrollBox_0 {
   widgetClassPath=&ScrollBox
   parent=HorizontalBox_0
   widget=ScrollBox_0
   properties {
    alwaysShowScrollbar=true
    widgetBarStyle {
     thickness=8
     draggedThumbImage {
      bIsDynamicallyLoaded=false
      drawAs=RoundedBox
      imageType=NoImage
      mirroring=NoMirror
      resourceName=None
      resourceObject=null
      tiling=NoTile
      imageSize=<0 0>
      margin=<ltrb 0 0 0 0>
      outlineSettings {
       bUseBrushTransparency=false
       roundingType=FixedRadius
       width=0
       color {colorUseRule=UseColor_Specified, specifiedColor=<rgba 0 0 0 0>}
       cornerRadii=<4 4 4 4>
      }
      tintColor {colorUseRule=UseColor_Specified, specifiedColor=<rgba 0.2159 0.2159 0.2159 1>}
      uVRegion {bIsValid=false, max=<0 0>, min=<0 0>}
     }
     horizontalBackgroundImage {
      bIsDynamicallyLoaded=false
      drawAs=NoDrawType
      imageType=NoImage
      mirroring=NoMirror
      resourceName=None
      resourceObject=null
      tiling=NoTile
      imageSize=<0 0>
      margin=<ltrb 0 0 0 0>
      outlineSettings {
       bUseBrushTransparency=false
       roundingType=HalfHeightRadius
       width=0
       color {colorUseRule=UseColor_Specified, specifiedColor=<rgba 0 0 0 0>}
       cornerRadii=<0 0 0 0>
      }
      tintColor {colorUseRule=UseColor_Specified, specifiedColor=<rgba 1 1 1 1>}
      uVRegion {bIsValid=false, max=<0 0>, min=<0 0>}
     }
     horizontalBottomSlotImage {
      bIsDynamicallyLoaded=false
      drawAs=NoDrawType
      imageType=NoImage
      mirroring=NoMirror
      resourceName=None
      resourceObject=null
      tiling=NoTile
      imageSize=<0 0>
      margin=<ltrb 0 0 0 0>
      outlineSettings {
       bUseBrushTransparency=false
       roundingType=HalfHeightRadius
       width=0
       color {colorUseRule=UseColor_Specified, specifiedColor=<rgba 0 0 0 0>}
       cornerRadii=<0 0 0 0>
      }
      tintColor {colorUseRule=UseColor_Specified, specifiedColor=<rgba 1 1 1 1>}
      uVRegion {bIsValid=false, max=<0 0>, min=<0 0>}
     }
     horizontalTopSlotImage {
      bIsDynamicallyLoaded=false
      drawAs=NoDrawType
      imageType=NoImage
      mirroring=NoMirror
      resourceName=None
      resourceObject=null
      tiling=NoTile
      imageSize=<0 0>
      margin=<ltrb 0 0 0 0>
      outlineSettings {
       bUseBrushTransparency=false
       roundingType=HalfHeightRadius
       width=0
       color {colorUseRule=UseColor_Specified, specifiedColor=<rgba 0 0 0 0>}
       cornerRadii=<0 0 0 0>
      }
      tintColor {colorUseRule=UseColor_Specified, specifiedColor=<rgba 1 1 1 1>}
      uVRegion {bIsValid=false, max=<0 0>, min=<0 0>}
     }
     hoveredThumbImage {
      bIsDynamicallyLoaded=false
      drawAs=RoundedBox
      imageType=NoImage
      mirroring=NoMirror
      resourceName=None
      resourceObject=null
      tiling=NoTile
      imageSize=<0 0>
      margin=<ltrb 0 0 0 0>
      outlineSettings {
       bUseBrushTransparency=false
       roundingType=FixedRadius
       width=0
       color {colorUseRule=UseColor_Specified, specifiedColor=<rgba 0 0 0 0>}
       cornerRadii=<4 4 4 4>
      }
      tintColor {colorUseRule=UseColor_Specified, specifiedColor=<rgba 0.2159 0.2159 0.2159 1>}
      uVRegion {bIsValid=false, max=<0 0>, min=<0 0>}
     }
     normalThumbImage {
      bIsDynamicallyLoaded=false
      drawAs=RoundedBox
      imageType=NoImage
      mirroring=NoMirror
      resourceName=None
      resourceObject=null
      tiling=NoTile
      imageSize=<0 0>
      margin=<ltrb 0 0 0 0>
      outlineSettings {
       bUseBrushTransparency=false
       roundingType=FixedRadius
       width=0
       color {colorUseRule=UseColor_Specified, specifiedColor=<rgba 0 0 0 0>}
       cornerRadii=<4 4 4 4>
      }
      tintColor {colorUseRule=UseColor_Specified, specifiedColor=<rgba 0.0953 0.0953 0.0953 1>}
      uVRegion {bIsValid=false, max=<0 0>, min=<0 0>}
     }
     verticalBackgroundImage {
      bIsDynamicallyLoaded=false
      drawAs=NoDrawType
      imageType=NoImage
      mirroring=NoMirror
      resourceName=None
      resourceObject=null
      tiling=NoTile
      imageSize=<0 0>
      margin=<ltrb 0 0 0 0>
      outlineSettings {
       bUseBrushTransparency=false
       roundingType=HalfHeightRadius
       width=0
       color {colorUseRule=UseColor_Specified, specifiedColor=<rgba 0 0 0 0>}
       cornerRadii=<0 0 0 0>
      }
      tintColor {colorUseRule=UseColor_Specified, specifiedColor=<rgba 1 1 1 1>}
      uVRegion {bIsValid=false, max=<0 0>, min=<0 0>}
     }
     verticalBottomSlotImage {
      bIsDynamicallyLoaded=false
      drawAs=NoDrawType
      imageType=NoImage
      mirroring=NoMirror
      resourceName=None
      resourceObject=null
      tiling=NoTile
      imageSize=<0 0>
      margin=<ltrb 0 0 0 0>
      outlineSettings {
       bUseBrushTransparency=false
       roundingType=HalfHeightRadius
       width=0
       color {colorUseRule=UseColor_Specified, specifiedColor=<rgba 0 0 0 0>}
       cornerRadii=<0 0 0 0>
      }
      tintColor {colorUseRule=UseColor_Specified, specifiedColor=<rgba 1 1 1 1>}
      uVRegion {bIsValid=false, max=<0 0>, min=<0 0>}
     }
     verticalTopSlotImage {
      bIsDynamicallyLoaded=false
      drawAs=NoDrawType
      imageType=NoImage
      mirroring=NoMirror
      resourceName=None
      resourceObject=null
      tiling=NoTile
      imageSize=<0 0>
      margin=<ltrb 0 0 0 0>
      outlineSettings {
       bUseBrushTransparency=false
       roundingType=HalfHeightRadius
       width=0
       color {colorUseRule=UseColor_Specified, specifiedColor=<rgba 0 0 0 0>}
       cornerRadii=<0 0 0 0>
      }
      tintColor {colorUseRule=UseColor_Specified, specifiedColor=<rgba 1 1 1 1>}
      uVRegion {bIsValid=false, max=<0 0>, min=<0 0>}
     }
    }
    widgetStyle {
     barThickness=8
     bottomShadowBrush {
      bIsDynamicallyLoaded=false
      drawAs=Box
      imageType=FullColor
      mirroring=NoMirror
      resourceName=../../../Engine/Content/Slate/Common/ScrollBoxShadowBottom.png
      resourceObject=null
      tiling=NoTile
      imageSize=<16 8>
      margin=<ltrb 0.5 0 0.5 1>
      outlineSettings {
       bUseBrushTransparency=false
       roundingType=HalfHeightRadius
       width=0
       color {colorUseRule=UseColor_Specified, specifiedColor=<rgba 0 0 0 0>}
       cornerRadii=<0 0 0 0>
      }
      tintColor {colorUseRule=UseColor_Specified, specifiedColor=<rgba 1 1 1 1>}
      uVRegion {bIsValid=false, max=<0 0>, min=<0 0>}
     }
     horizontalScrolledContentPadding=<ltrb 0 0 1 0>
     leftShadowBrush {
      bIsDynamicallyLoaded=false
      drawAs=Box
      imageType=FullColor
      mirroring=NoMirror
      resourceName=../../../Engine/Content/Slate/Common/ScrollBoxShadowLeft.png
      resourceObject=null
      tiling=NoTile
      imageSize=<8 16>
      margin=<ltrb 1 0.5 0 0.5>
      outlineSettings {
       bUseBrushTransparency=false
       roundingType=HalfHeightRadius
       width=0
       color {colorUseRule=UseColor_Specified, specifiedColor=<rgba 0 0 0 0>}
       cornerRadii=<0 0 0 0>
      }
      tintColor {colorUseRule=UseColor_Specified, specifiedColor=<rgba 1 1 1 1>}
      uVRegion {bIsValid=false, max=<0 0>, min=<0 0>}
     }
     rightShadowBrush {
      bIsDynamicallyLoaded=false
      drawAs=Box
      imageType=FullColor
      mirroring=NoMirror
      resourceName=../../../Engine/Content/Slate/Common/ScrollBoxShadowRight.png
      resourceObject=null
      tiling=NoTile
      imageSize=<8 16>
      margin=<ltrb 0 0.5 1 0.5>
      outlineSettings {
       bUseBrushTransparency=false
       roundingType=HalfHeightRadius
       width=0
       color {colorUseRule=UseColor_Specified, specifiedColor=<rgba 0 0 0 0>}
       cornerRadii=<0 0 0 0>
      }
      tintColor {colorUseRule=UseColor_Specified, specifiedColor=<rgba 1 1 1 1>}
      uVRegion {bIsValid=false, max=<0 0>, min=<0 0>}
     }
     topShadowBrush {
      bIsDynamicallyLoaded=false
      drawAs=Box
      imageType=FullColor
      mirroring=NoMirror
      resourceName=../../../Engine/Content/Slate/Common/ScrollBoxShadowTop.png
      resourceObject=null
      tiling=NoTile
      imageSize=<16 8>
      margin=<ltrb 0.5 1 0.5 0>
      outlineSettings {
       bUseBrushTransparency=false
       roundingType=HalfHeightRadius
       width=0
       color {colorUseRule=UseColor_Specified, specifiedColor=<rgba 0 0 0 0>}
       cornerRadii=<0 0 0 0>
      }
      tintColor {colorUseRule=UseColor_Specified, specifiedColor=<rgba 1 1 1 1>}
      uVRegion {bIsValid=false, max=<0 0>, min=<0 0>}
     }
     verticalScrolledContentPadding=<ltrb 0 0 0 1>
    }
   }
   slot {class=HorizontalBoxSlot, properties {size {sizeRule=Fill, value=1}}}
  }
  Emote {
   widgetClassPath=&DetailsView
   parent=ScrollBox_0
   widget=Emote
   properties {categoriesToShow [Emote]}
   slot {class=ScrollBoxSlot}
  }
  CurrentArrayPos {
   widgetClassPath=&DetailsView
   parent=ScrollBox_0
   widget=CurrentArrayPos
   properties {categoriesToShow [CurrentArrayPosition]}
   slot {class=ScrollBoxSlot}
  }
  ScrollBox_2 {
   widgetClassPath=&ScrollBox
   parent=ScrollBox_0
   widget=ScrollBox_2
   properties {
    widgetBarStyle {
     thickness=8
     draggedThumbImage {
      bIsDynamicallyLoaded=false
      drawAs=RoundedBox
      imageType=NoImage
      mirroring=NoMirror
      resourceName=None
      resourceObject=null
      tiling=NoTile
      imageSize=<0 0>
      margin=<ltrb 0 0 0 0>
      outlineSettings {
       bUseBrushTransparency=false
       roundingType=FixedRadius
       width=0
       color {colorUseRule=UseColor_Specified, specifiedColor=<rgba 0 0 0 0>}
       cornerRadii=<4 4 4 4>
      }
      tintColor {colorUseRule=UseColor_Specified, specifiedColor=<rgba 0.2159 0.2159 0.2159 1>}
      uVRegion {bIsValid=false, max=<0 0>, min=<0 0>}
     }
     horizontalBackgroundImage {
      bIsDynamicallyLoaded=false
      drawAs=NoDrawType
      imageType=NoImage
      mirroring=NoMirror
      resourceName=None
      resourceObject=null
      tiling=NoTile
      imageSize=<0 0>
      margin=<ltrb 0 0 0 0>
      outlineSettings {
       bUseBrushTransparency=false
       roundingType=HalfHeightRadius
       width=0
       color {colorUseRule=UseColor_Specified, specifiedColor=<rgba 0 0 0 0>}
       cornerRadii=<0 0 0 0>
      }
      tintColor {colorUseRule=UseColor_Specified, specifiedColor=<rgba 1 1 1 1>}
      uVRegion {bIsValid=false, max=<0 0>, min=<0 0>}
     }
     horizontalBottomSlotImage {
      bIsDynamicallyLoaded=false
      drawAs=NoDrawType
      imageType=NoImage
      mirroring=NoMirror
      resourceName=None
      resourceObject=null
      tiling=NoTile
      imageSize=<0 0>
      margin=<ltrb 0 0 0 0>
      outlineSettings {
       bUseBrushTransparency=false
       roundingType=HalfHeightRadius
       width=0
       color {colorUseRule=UseColor_Specified, specifiedColor=<rgba 0 0 0 0>}
       cornerRadii=<0 0 0 0>
      }
      tintColor {colorUseRule=UseColor_Specified, specifiedColor=<rgba 1 1 1 1>}
      uVRegion {bIsValid=false, max=<0 0>, min=<0 0>}
     }
     horizontalTopSlotImage {
      bIsDynamicallyLoaded=false
      drawAs=NoDrawType
      imageType=NoImage
      mirroring=NoMirror
      resourceName=None
      resourceObject=null
      tiling=NoTile
      imageSize=<0 0>
      margin=<ltrb 0 0 0 0>
      outlineSettings {
       bUseBrushTransparency=false
       roundingType=HalfHeightRadius
       width=0
       color {colorUseRule=UseColor_Specified, specifiedColor=<rgba 0 0 0 0>}
       cornerRadii=<0 0 0 0>
      }
      tintColor {colorUseRule=UseColor_Specified, specifiedColor=<rgba 1 1 1 1>}
      uVRegion {bIsValid=false, max=<0 0>, min=<0 0>}
     }
     hoveredThumbImage {
      bIsDynamicallyLoaded=false
      drawAs=RoundedBox
      imageType=NoImage
      mirroring=NoMirror
      resourceName=None
      resourceObject=null
      tiling=NoTile
      imageSize=<0 0>
      margin=<ltrb 0 0 0 0>
      outlineSettings {
       bUseBrushTransparency=false
       roundingType=FixedRadius
       width=0
       color {colorUseRule=UseColor_Specified, specifiedColor=<rgba 0 0 0 0>}
       cornerRadii=<4 4 4 4>
      }
      tintColor {colorUseRule=UseColor_Specified, specifiedColor=<rgba 0.2159 0.2159 0.2159 1>}
      uVRegion {bIsValid=false, max=<0 0>, min=<0 0>}
     }
     normalThumbImage {
      bIsDynamicallyLoaded=false
      drawAs=RoundedBox
      imageType=NoImage
      mirroring=NoMirror
      resourceName=None
      resourceObject=null
      tiling=NoTile
      imageSize=<0 0>
      margin=<ltrb 0 0 0 0>
      outlineSettings {
       bUseBrushTransparency=false
       roundingType=FixedRadius
       width=0
       color {colorUseRule=UseColor_Specified, specifiedColor=<rgba 0 0 0 0>}
       cornerRadii=<4 4 4 4>
      }
      tintColor {colorUseRule=UseColor_Specified, specifiedColor=<rgba 0.0953 0.0953 0.0953 1>}
      uVRegion {bIsValid=false, max=<0 0>, min=<0 0>}
     }
     verticalBackgroundImage {
      bIsDynamicallyLoaded=false
      drawAs=NoDrawType
      imageType=NoImage
      mirroring=NoMirror
      resourceName=None
      resourceObject=null
      tiling=NoTile
      imageSize=<0 0>
      margin=<ltrb 0 0 0 0>
      outlineSettings {
       bUseBrushTransparency=false
       roundingType=HalfHeightRadius
       width=0
       color {colorUseRule=UseColor_Specified, specifiedColor=<rgba 0 0 0 0>}
       cornerRadii=<0 0 0 0>
      }
      tintColor {colorUseRule=UseColor_Specified, specifiedColor=<rgba 1 1 1 1>}
      uVRegion {bIsValid=false, max=<0 0>, min=<0 0>}
     }
     verticalBottomSlotImage {
      bIsDynamicallyLoaded=false
      drawAs=NoDrawType
      imageType=NoImage
      mirroring=NoMirror
      resourceName=None
      resourceObject=null
      tiling=NoTile
      imageSize=<0 0>
      margin=<ltrb 0 0 0 0>
      outlineSettings {
       bUseBrushTransparency=false
       roundingType=HalfHeightRadius
       width=0
       color {colorUseRule=UseColor_Specified, specifiedColor=<rgba 0 0 0 0>}
       cornerRadii=<0 0 0 0>
      }
      tintColor {colorUseRule=UseColor_Specified, specifiedColor=<rgba 1 1 1 1>}
      uVRegion {bIsValid=false, max=<0 0>, min=<0 0>}
     }
     verticalTopSlotImage {
      bIsDynamicallyLoaded=false
      drawAs=NoDrawType
      imageType=NoImage
      mirroring=NoMirror
      resourceName=None
      resourceObject=null
      tiling=NoTile
      imageSize=<0 0>
      margin=<ltrb 0 0 0 0>
      outlineSettings {
       bUseBrushTransparency=false
       roundingType=HalfHeightRadius
       width=0
       color {colorUseRule=UseColor_Specified, specifiedColor=<rgba 0 0 0 0>}
       cornerRadii=<0 0 0 0>
      }
      tintColor {colorUseRule=UseColor_Specified, specifiedColor=<rgba 1 1 1 1>}
      uVRegion {bIsValid=false, max=<0 0>, min=<0 0>}
     }
    }
    widgetStyle {
     barThickness=8
     bottomShadowBrush {
      bIsDynamicallyLoaded=false
      drawAs=Box
      imageType=FullColor
      mirroring=NoMirror
      resourceName=../../../Engine/Content/Slate/Common/ScrollBoxShadowBottom.png
      resourceObject=null
      tiling=NoTile
      imageSize=<16 8>
      margin=<ltrb 0.5 0 0.5 1>
      outlineSettings {
       bUseBrushTransparency=false
       roundingType=HalfHeightRadius
       width=0
       color {colorUseRule=UseColor_Specified, specifiedColor=<rgba 0 0 0 0>}
       cornerRadii=<0 0 0 0>
      }
      tintColor {colorUseRule=UseColor_Specified, specifiedColor=<rgba 1 1 1 1>}
      uVRegion {bIsValid=false, max=<0 0>, min=<0 0>}
     }
     horizontalScrolledContentPadding=<ltrb 0 0 1 0>
     leftShadowBrush {
      bIsDynamicallyLoaded=false
      drawAs=Box
      imageType=FullColor
      mirroring=NoMirror
      resourceName=../../../Engine/Content/Slate/Common/ScrollBoxShadowLeft.png
      resourceObject=null
      tiling=NoTile
      imageSize=<8 16>
      margin=<ltrb 1 0.5 0 0.5>
      outlineSettings {
       bUseBrushTransparency=false
       roundingType=HalfHeightRadius
       width=0
       color {colorUseRule=UseColor_Specified, specifiedColor=<rgba 0 0 0 0>}
       cornerRadii=<0 0 0 0>
      }
      tintColor {colorUseRule=UseColor_Specified, specifiedColor=<rgba 1 1 1 1>}
      uVRegion {bIsValid=false, max=<0 0>, min=<0 0>}
     }
     rightShadowBrush {
      bIsDynamicallyLoaded=false
      drawAs=Box
      imageType=FullColor
      mirroring=NoMirror
      resourceName=../../../Engine/Content/Slate/Common/ScrollBoxShadowRight.png
      resourceObject=null
      tiling=NoTile
      imageSize=<8 16>
      margin=<ltrb 0 0.5 1 0.5>
      outlineSettings {
       bUseBrushTransparency=false
       roundingType=HalfHeightRadius
       width=0
       color {colorUseRule=UseColor_Specified, specifiedColor=<rgba 0 0 0 0>}
       cornerRadii=<0 0 0 0>
      }
      tintColor {colorUseRule=UseColor_Specified, specifiedColor=<rgba 1 1 1 1>}
      uVRegion {bIsValid=false, max=<0 0>, min=<0 0>}
     }
     topShadowBrush {
      bIsDynamicallyLoaded=false
      drawAs=Box
      imageType=FullColor
      mirroring=NoMirror
      resourceName=../../../Engine/Content/Slate/Common/ScrollBoxShadowTop.png
      resourceObject=null
      tiling=NoTile
      imageSize=<16 8>
      margin=<ltrb 0.5 1 0.5 0>
      outlineSettings {
       bUseBrushTransparency=false
       roundingType=HalfHeightRadius
       width=0
       color {colorUseRule=UseColor_Specified, specifiedColor=<rgba 0 0 0 0>}
       cornerRadii=<0 0 0 0>
      }
      tintColor {colorUseRule=UseColor_Specified, specifiedColor=<rgba 1 1 1 1>}
      uVRegion {bIsValid=false, max=<0 0>, min=<0 0>}
     }
     verticalScrolledContentPadding=<ltrb 0 0 0 1>
    }
   }
   slot {class=ScrollBoxSlot, properties {size {sizeRule=Fill, value=1}}}
  }
  AllCIDS {
   widgetClassPath=&DetailsView
   parent=ScrollBox_2
   widget=AllCIDS
   properties {categoriesToShow [CidList]}
   slot {class=ScrollBoxSlot, properties {size {sizeRule=Fill, value=1}}}
  }
  CosmeticPickerv2 {
   widgetClassPath=/Game/NeverCook/ArtTools/Tech_Art/CosmeticPicker/CosmeticPickerv2
   bIsVariable=true
   parent=VerticalBox_0
   widget=CosmeticPickerv2
   slot {class=VerticalBoxSlot, properties {size {sizeRule=Fill, value=1}}}
  }
 ]
}
