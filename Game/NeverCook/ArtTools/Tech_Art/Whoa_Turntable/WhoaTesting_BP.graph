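# Editor utility Blueprint in the Whoa Turntable tech art tools that pushes label text into a turntable UI widget.
# The custom Update Text event casts the Widget component's user widget to WBP_Turntable_UI and calls setText with name, date, build and object list.
# Pass, FrameNumber and Version are left unset, the UI variable is unused, and the standard events and construction script are empty.
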EditorUtilityBlueprint /Game/NeverCook/ArtTools/Tech_Art/Whoa_Turntable/WhoaTesting_BP
graphs [
 EventGraph {nodes [
   n1 = Update Text!() -> (OutputDelegate, then, Name:String, Date:String, build:String, objectList:String) {is=title K2Node_CustomEvent}
   > n2 = Cast To WBP_Turntable_UI!(
    execute
    Object=GetWidget!(self:Widget Component Object Reference=Get Widget() -> (Widget:Widget Component Object Reference)) -> (ReturnValue:User Widget Object Reference)
   ) -> (then, CastFailed, AsWBP Turntable UI:WBP Turntable UI Object Reference) {is=title K2Node_DynamicCast}
   > setText!(
    execute
    self:WBP Turntable UI Object Reference=@n2
    Name:String=@n1.Name
    Date:String=@n1.Date
    build:String=@n1.build
    objectList:String=@n1.objectList
    Pass:String
    FrameNumber:String
    Version:String
   ) -> (then)
   Event BeginPlay()
   Event ActorBeginOverlap()
   Event Tick!() -> (OutputDelegate, then, DeltaSeconds:float)
  ]}
 UserConstructionScript {nodes [Construction Script()]}
]
variables [UI {type=WBP_Turntable_UI_C, default=null}]
