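# Test Level Sequence in the Whoa_Turntable tech art tools folder that sets up a static cine camera shot.
# A CineCameraActor with a spawn track holds a fixed transform, with no keyframes shown.
# Its CineCameraComponent child has tracks holding focal length, manual focus distance and aperture at constant values.
# A camera cut track spans a 150-frame range at a 30 fps display rate.
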LevelSequence /Game/NeverCook/ArtTools/Tech_Art/Whoa_Turntable/WoahTestingLevelSequence
bindings [
 CineCameraActor {
  id=#1
  possessed_class=CineCameraActor
  tracks [
   {
    type=MovieScene3DTransformTrack
    sections [{
      type=MovieScene3DTransformSection
      channels [
       Location.X {type=MovieSceneScriptingDoubleChannel, default=-104.1062}
       Location.Y {type=MovieSceneScriptingDoubleChannel, default=-1754.2262}
       Location.Z {type=MovieSceneScriptingDoubleChannel, default=2269.7808}
       Rotation.X {type=MovieSceneScriptingDoubleChannel, default=0}
       Rotation.Y {type=MovieSceneScriptingDoubleChannel, default=-5}
       Rotation.Z {type=MovieSceneScriptingDoubleChannel, default=-120.1823}
       Scale.X {type=MovieSceneScriptingDoubleChannel, default=1}
       Scale.Y {type=MovieSceneScriptingDoubleChannel, default=1}
       Scale.Z {type=MovieSceneScriptingDoubleChannel, default=1}
      ]
     }]
   }
   {type=MovieSceneSpawnTrack, sections [{type=MovieSceneSpawnSection, channels [None {type=MovieSceneScriptingBoolChannel, default=true}]}]}
  ]
 }
 CameraComponent {
  id=#2
  parent_id=#1
  possessed_class=CineCameraComponent
  tracks [
   {
    type=MovieSceneFloatTrack
    display_name=Current Focal Length
    sections [{type=MovieSceneFloatSection, channels [None {type=MovieSceneScriptingFloatChannel, default=35}]}]
   }
   {
    type=MovieSceneFloatTrack
    display_name="Manual Focus Distance (Focus Settings)"
    sections [{type=MovieSceneFloatSection, channels [None {type=MovieSceneScriptingFloatChannel, default=100000}]}]
   }
   {
    type=MovieSceneFloatTrack
    display_name=Current Aperture
    sections [{type=MovieSceneFloatSection, channels [None {type=MovieSceneScriptingFloatChannel, default=2.8}]}]
   }
  ]
 }
]
display_rate {denominator=1, numerator=30}
tick_resolution {denominator=1, numerator=24000}
tracks [{type=MovieSceneCameraCutTrack, sections [{type=MovieSceneCameraCutSection, end_frame=150, start_frame=0}]}]
